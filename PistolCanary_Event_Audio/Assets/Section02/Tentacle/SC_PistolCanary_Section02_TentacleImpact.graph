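# Sound cue for tentacle impact sounds in Section 02 of the PistolCanary event.
# A Random node picks one of eight equally weighted impact waves and feeds a Modulator that randomizes its volume.
# The cue routes to the Fort_Events sound class and uses a dedicated tentacle impact attenuation asset for long-range falloff.
# A concurrency override caps simultaneous voices and sets a short retrigger time.
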SoundCue /PistolCanary_Event_Audio/Assets/Section02/Tentacle/SC_PistolCanary_Section02_TentacleImpact
root=Modulator_0
nodes [
 {id=Modulator_0, children [Random_1], properties {volumeMax=0.9, volumeMin=0.6}}
 {
  id=Random_1
  children [WavePlayer_0, WavePlayer_1, WavePlayer_2, WavePlayer_3, WavePlayer_8, WavePlayer_9, WavePlayer_10, WavePlayer_11]
  properties {weights [1,1,1,1,1,1,1,1]}
 }
 {id=WavePlayer_0, properties {soundWaveAssetPtr=/PistolCanary_Event_Audio/Assets/Section02/wav/PistolCanary_Tentacle_Impact_01}}
 {id=WavePlayer_1, properties {soundWaveAssetPtr=/PistolCanary_Event_Audio/Assets/Section02/wav/PistolCanary_Tentacle_Impact_02}}
 {id=WavePlayer_10, properties {soundWaveAssetPtr=/PistolCanary_Event_Audio/Assets/Section02/wav/PistolCanary_Tentacle_Impact_07}}
 {id=WavePlayer_11, properties {soundWaveAssetPtr=/PistolCanary_Event_Audio/Assets/Section02/wav/PistolCanary_Tentacle_Impact_08}}
 {id=WavePlayer_2, properties {soundWaveAssetPtr=/PistolCanary_Event_Audio/Assets/Section02/wav/PistolCanary_Tentacle_Impact_03}}
 {id=WavePlayer_3, properties {soundWaveAssetPtr=/PistolCanary_Event_Audio/Assets/Section02/wav/PistolCanary_Tentacle_Impact_04}}
 {id=WavePlayer_8, properties {soundWaveAssetPtr=/PistolCanary_Event_Audio/Assets/Section02/wav/PistolCanary_Tentacle_Impact_05}}
 {id=WavePlayer_9, properties {soundWaveAssetPtr=/PistolCanary_Event_Audio/Assets/Section02/wav/PistolCanary_Tentacle_Impact_06}}
]
properties {
 attenuationSettings=/PistolCanary_Event_Audio/Assets/Settings/ATT_PistolCanary_Tentacle_Impacts
 bOverrideConcurrency=true
 maxDistance=23000
 soundClassObject=&Fort_Events
 volumeMultiplier=0.6
 concurrencyOverrides {maxCount=4, retriggerTime=0.3, voiceStealReleaseTime=0.1, volumeScaleAttackTime=0.1}
}
